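# Looping winter wind ambience cue for the Asteria IceCavern POI, routed to the Athena outdoor ambient sound class.
# A QualityLevel root selects either a three-layer mix or one of three cheaper single-loop Spooky_Night02 variants.
# The full mix blends winter day, spooky night and desert night wind loops, using 100-second looping Envelopers, partly driven by shared external curves, to drift their levels.
# The single-loop branches use 30-second envelopes with randomized pitch and volume.
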SoundCue /Game/Sounds/Asteria/POI/IceCavern/Ambience_Winter_Wind_Day_Loop_01_Cue
root=QualityLevel_0
nodes [
 {
  id=Enveloper_0
  children [WavePlayer_2]
  properties {
   bLoop=true
   bLoopIndefinitely=true
   loopEnd=30
   pitchCurve {editorCurveData {keys [{interpMode=RCIM_Cubic, time=0}, <at 15.9468 0.9181 RCIM_Cubic>, {interpMode=RCIM_Cubic, time=30}]}}
   volumeCurve {editorCurveData {keys [<at 0 1 RCIM_Cubic>,<at 16.1421 0.921 RCIM_Cubic>,<at 30 1 RCIM_Cubic>]}}
  }
 }
 {
  id=Enveloper_1
  children [WavePlayer_0]
  properties {
   bLoop=true
   bLoopIndefinitely=true
   loopEnd=30
   pitchCurve {editorCurveData {keys [{interpMode=RCIM_Cubic, time=0}, <at 15.9468 0.9181 RCIM_Cubic>, {interpMode=RCIM_Cubic, time=30}]}}
   volumeCurve {editorCurveData {keys [<at 0 1 RCIM_Cubic>,<at 16.1421 0.921 RCIM_Cubic>,<at 30 1 RCIM_Cubic>]}}
  }
 }
 {
  id=Enveloper_2
  children [WavePlayer_4]
  properties {
   bLoop=true
   bLoopIndefinitely=true
   loopEnd=30
   pitchCurve {editorCurveData {keys [{interpMode=RCIM_Cubic, time=0}, <at 15.9468 0.9181 RCIM_Cubic>, {interpMode=RCIM_Cubic, time=30}]}}
   volumeCurve {editorCurveData {keys [<at 0 1 RCIM_Cubic>,<at 16.1421 0.921 RCIM_Cubic>,<at 30 1 RCIM_Cubic>]}}
  }
 }
 {
  id=Enveloper_3
  children [WavePlayer_1]
  properties {
   bLoop=true
   bLoopIndefinitely=true
   loopEnd=100
   pitchCurve {externalCurve=/Game/Sounds/Trans_Enveloper_Curves/PitchRND_100s_02, editorCurveData {keys [{time=0}]}}
   volumeCurve {editorCurveData {keys [<at 0 0.5 RCIM_Cubic>,<at 50 1 RCIM_Cubic>,<at 100 0.5 RCIM_Cubic>]}}
  }
 }
 {
  id=Enveloper_6
  children [Modulator_11]
  properties {
   bLoop=true
   bLoopIndefinitely=true
   loopEnd=100
   pitchCurve {editorCurveData {keys [<at 0 0.8>]}}
   volumeCurve {
    externalCurve=/Game/Sounds/Trans_Enveloper_Curves/BellReversed_100s.BellReversed_100S
    editorCurveData {keys [<at 0 1 RCIM_Cubic>,<at 50 0.5 RCIM_Cubic>,<at 100 1 RCIM_Cubic>]}
   }
  }
 }
 {
  id=Enveloper_7
  children [Modulator_12]
  properties {
   bLoop=true
   bLoopIndefinitely=true
   loopEnd=100
   pitchCurve {editorCurveData {keys [{time=0}]}}
   volumeCurve {
    externalCurve=/Game/Sounds/Trans_Enveloper_Curves/Bell_100s
    editorCurveData {keys [<at 0 0.1 RCIM_Cubic>,<at 100 1 RCIM_Cubic>,<at 200 0.1 RCIM_Cubic>]}
   }
  }
 }
 {id=Mixer_1, children [Enveloper_6,Enveloper_3,Enveloper_7], properties {inputVolume [1,0.25,0.4]}}
 {id=Modulator_0, children [Enveloper_0], properties {pitchMax=1.1, pitchMin=0.9, volumeMax=1.1, volumeMin=0.9}}
 {id=Modulator_1, children [Enveloper_1], properties {pitchMax=1.1, pitchMin=0.9, volumeMax=1.1, volumeMin=0.9}}
 {id=Modulator_11, children [WavePlayer_3], properties {pitchMax=1, pitchMin=1}}
 {id=Modulator_12, children [WavePlayer_33], properties {pitchMax=0.9, pitchMin=0.9}}
 {id=Modulator_2, children [Enveloper_2], properties {pitchMax=1.1, pitchMin=0.9, volumeMax=1.1, volumeMin=0.9}}
 {id=Modulator_23, children [Mixer_1], properties {pitchMax=1, pitchMin=1, volumeMax=1, volumeMin=1}}
 {id=QualityLevel_0, children [Modulator_23,Modulator_0,Modulator_1,Modulator_2]}
 {id=WavePlayer_0, properties {bLooping=true, soundWaveAssetPtr=/Game/Sounds/Fort_Ambient_Looping/Dread_Exterior/Spooky_Night02_opt}}
 {id=WavePlayer_1, properties {bLooping=true, soundWaveAssetPtr=/Game/Sounds/Fort_Ambient_Looping/Dread_Exterior/Spooky_Night02}}
 {id=WavePlayer_2, properties {bLooping=true, soundWaveAssetPtr=&Spooky_Night02}}
 {id=WavePlayer_3, properties {bLooping=true, soundWaveAssetPtr=/Game/Sounds/Fort_Ambient_Looping/Winter/Amb_Winter_Day_Loop}}
 {id=WavePlayer_33, properties {bLooping=true, soundWaveAssetPtr=/Game/Sounds/Fort_Ambient_Looping/Desert_Exterior/desert_wind_Night}}
 {id=WavePlayer_4, properties {bLooping=true, soundWaveAssetPtr=&Spooky_Night02_opt}}
]
properties {
 bOverrideConcurrency=true
 maxDistance=2097152
 soundClassObject=&Fort_Ambient_Outdoors_Athena
 volumeMultiplier=0.501
}
